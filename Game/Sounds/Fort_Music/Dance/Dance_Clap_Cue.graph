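# Sound cue for the Clap dance emote audio, playing the dance_soldier_clap wave.
# A single wave player is the root, with no randomization, mixing or looping nodes.
# It routes through the Fort_Music_Emote sound class and uses the shared Fort_Emotes_Athena_Att attenuation for spatial falloff.
# Its own concurrency override prevents new instances from starting once the limit is reached, instead of stopping ones already playing.
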SoundCue /Game/Sounds/Fort_Music/Dance/Dance_Clap_Cue
root=WavePlayer_37
nodes [{id=WavePlayer_37, properties {soundWaveAssetPtr=&/dance_soldier_clap}}]
properties {
 attenuationSettings=&Fort_Emotes_Athena_Att
 bOverrideConcurrency=true
 maxDistance=1600
 soundClassObject=&Fort_Music_Emote
 volumeMultiplier=1
 concurrencyOverrides {resolutionRule=PreventNew}
}
